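# Niagara system for the BandSoda emote, emitting rising soda blobs and bubbles, with glow and bubble colors selected by User.ColorIndex.
# Metaball1 reads back ten of its own particles, packs blob, second-blob, capsule and ground data, apparently for metaball rendering, and tests whether the camera is inside bounds.
# SmallBubble spawns platform-scaled sprite counts from Metaball1 particles and tracks them via attribute readers.
# FallBack provides a simpler mesh-and-sprite version of the blobs.
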NiagaraSystem /CosmeticCompanions/FX/Builtins/BandSoda/Emote/NS_Emote_BandSoda_Metaballs
dependencies {
 usedDataInterfaces [
  /Script/Niagara.NiagaraDataInterfaceSimpleCounter
  &NiagaraDataInterfaceParticleRead
  &NiagaraDataInterfacePlatformSet
  &NiagaraDataInterfaceCurve
  &NiagaraDataInterfaceCamera
  &NiagaraDataInterfaceCollisionQuery
  &NiagaraDataInterfaceColorCurve
 ]
 usedDynamicInputs [&Multiply_FloatByInt]
 usedModules [
  &SystemState
  &self:SystemSpawnScript.SystemScriptSource.SystemScriptGraph.NiagaraNodeAssignment_1.SetVariables_#1
  &self:SystemSpawnScript.SystemScriptSource.SystemScriptGraph.NiagaraNodeAssignment_2.SetVariables_#2
  &self:SystemSpawnScript.SystemScriptSource.SystemScriptGraph.NiagaraNodeAssignment_0.SetVariables_#3
  &self:SystemSpawnScript.SystemScriptSource.SystemScriptGraph.NiagaraNodeAssignment_3.SetVariables_#4
 ]
}
emitters {
 FallBack {
  emitterSpawnScript [
   /SetVariables(Emitter.Offset:Vector3f=<5 -20 0>) {isSetParameters=true}
   /SetVariables(Emitter.Offset_2:Vector3f=<5 -20 0>) {isSetParameters=true}
  ]
  emitterUpdateScript [
   &EmitterState()
   &SpawnRate(SpawnRate=&Multiply_Float(A=1.6, B=$System.SpawnRateScale), Use Spawn Probability=false)
  ]
  particleSpawnScript [
   &PartitionParticles()
   &InitializeParticle(
    Color=&FN_Select_Color_from_Time_Of_Day(Morning=<rgba 0.8698 0.7989 0.7835 1>, Day=<rgba 1 1 1 1>, Evening=<rgba 0.9062 0.8324 0.8163 1>, Night=<rgba 0.6146 0.6146 0.6146 1>)
    Lifetime=&RandomRangeFloat(Minimum=4.3, Maximum=4.7)
    Mesh Scale Mode=::NewEnumerator1
    Mesh Uniform Scale Max=0.16
    Mesh Uniform Scale Min=0.09
    Sprite Size Mode=::NewEnumerator3
    Uniform Sprite Size=&RandomRangeFloat(Minimum=3, Maximum=9)
    UsePositionOffset=false
    Write Material Random=true
   )
   &ShapeLocation(
    Handle Radius=0
    Large Radius=45
    MaskSpawnGroup=false
    Offset=&MergeHelper_VectorSwitch(0=$Emitter.Offset, 1=$Emitter.Offset_2, 2=?, 3=?, Index=$Particles.PartitionParticles.Partition, Num=::NewEnumerator0)
    Offset Mode=::NewEnumerator0
    Rotation Angle:NiagaraFloat=&MakeCustomFloatFromBool(
     Bool=&SetBoolByIntComparison(A=$Particles.PartitionParticles.Partition, B=0, Comparison Type=::NewEnumerator4)
     True Float=0
     False Float=0.5
    )
    Rotation Axis=<0 0 1>
    Rotation Mode=::NewEnumerator1
    Shape Primitive=::NewEnumerator3
    Torus Distribution Mode:ENiagaraTorusDistributionMode=::NewEnumerator1
    U Position=&Divide_Float(
     A=/Niagara/DynamicInputs/Math/Floor_Float(Float:NiagaraFloat=&Multiply_Float(A=&RandomRangeFloat(Minimum=0.3, Maximum=0.6), B=10))
     B=10
    )
    UseOwnerScale=true
   )
   &AddVelocity(Velocity=&RandomRangeVector(Minimum=<0 0 30>, Maximum=<2 2 50>))
   &self:TraceStartPos(DownOffset=128) {module=TraceStartPos}
  ]
  particleUpdateScript [
   &ParticleState()
   &UpdateMeshOrientation(
    Rotation Rate=&RandomRangeFloat(Minimum=0.1, Maximum=0.2)
    Rotation Vector=&RandomVector(Vector Scale=1)
   )
   &ScaleSpriteSize(
    Uniform Curve Scale=$System.TotalScale
    Uniform Curve Sprite Scale=DataInterface() {propertyValues {Curve {
       defaultValue=1
       keys [
        {arriveTangent=9.0725, interpMode=RCIM_Cubic, leaveTangent=9.0725, tangentMode=RCTM_User, time=0, value=0.25}
        {arriveTangent=0.5491, interpMode=RCIM_Cubic, leaveTangent=0.5491, tangentMode=RCTM_User, time=0.13, value=0.9847}
        {arriveTangent=0.0679, interpMode=RCIM_Cubic, leaveTangent=0.0679, tangentMode=RCTM_User, time=0.77, value=0.9653}
        {arriveTangent=-3.2218, interpMode=RCIM_Cubic, leaveTangent=-3.2218, tangentMode=RCTM_User, time=1, value=0}
       ]
      }}}
   )
   &ScaleMeshSize(Scale Factor=&MakeVector(
     X=1
     Y=&Lerp_Float(
      A=1
      B=1.3
      Alpha=&Waveform(
       "Waveform [1]"=::NewEnumerator0
       "[1] Frequency"=&RandomRangeFloat(Minimum=0.25, Maximum=0.5)
       "[1] Amplitude Min/Max"=<0 1>
      )
     )
     Z=1
    )) {module=ScaleMeshSize001}
   &ScaleMeshSize(Scale Factor=&VectorFromFloat(Value=&FloatFromCurve(
      FloatCurve=DataInterface() {propertyValues {Curve {
         defaultValue=1
         keys [
          {arriveTangent=9.0725, interpMode=RCIM_Cubic, leaveTangent=9.0725, tangentMode=RCTM_User, time=0, value=0}
          {arriveTangent=0.5491, interpMode=RCIM_Cubic, leaveTangent=0.5491, tangentMode=RCTM_User, time=0.13, value=0.9847}
          {arriveTangent=0.0679, interpMode=RCIM_Cubic, leaveTangent=0.0679, tangentMode=RCTM_User, time=0.77, value=1}
          {arriveTangent=-3.2218, interpMode=RCIM_Cubic, leaveTangent=-3.2218, tangentMode=RCTM_User, time=1, value=0}
         ]
        }}}
      Scale Curve=$System.TotalScale
     )))
   &Drag(Drag=&RandomRangeFloat(Minimum=2, Maximum=2.5), Use Linear Drag=true, Use Rotational Drag=true)
   &CurlNoiseForce(Bias Noise=false, Noise Frequency=20, Pan Noise=true, Pan Noise Field=<0 0 0.6>)
   &AccelerationForce(Acceleration=<0 0 40>)
   &SolveForcesAndVelocity(Clamp Velocity=false, Limit Acceleration=false)
   &ScaleColor(
    Scale Alpha=&FloatFromCurve(FloatCurve=DataInterface() {propertyValues {Curve {keys [{interpMode=RCIM_Cubic, tangentMode=RCTM_Break, time=0, value=0}, {interpMode=RCIM_Cubic, tangentMode=RCTM_Break, time=0.3, value=1}]}}})
    ScaleA=true
    ScaleRGB=true
   )
   /SetVariables(Particles.PriteSizeForSoftGlow:Vector2f=&Multiply_Vector2D(A=$Particles.SpriteSize, B=<8 8>)) {isSetParameters=true}
   /SetVariables(Particles.Linear Color2:LinearColor=&MultiplyLinearColor(
     Linear Color=&MergeHelper_ColorSwitch(
      0=<rgba 0.1 0.72 1 0.15>
      1=<rgba 0.1 1 0.4976 0.15>
      2=<rgba 0.55 0.1 1 0.25>
      3=<rgba 1 0.1 0.1 0.15>
      4=<rgba 0 0 0 0>
      Index=$User.ColorIndex
      Num=::NewEnumerator3
     )
     Scale Factor=&MakeVector4_FromLinearColor(LinearColor=&ColorFromCurve(ColorCurve=DataInterface() {propertyValues {
         AlphaCurve {keys [<at 0 0>,<at 0.1244 0.4088>,<at 0.6394 0.9356>,<at 0.8752 0.4726>,<at 1 0>]}
         BlueCurve {keys [<at 0 1>,<at 1 1>]}
         GreenCurve {keys [<at 0 1>,<at 1 1>]}
         RedCurve {keys [<at 0 1>,<at 1 1>]}
        }}))
    )) {isSetParameters=true}
  ]
  rendererClasses [&NiagaraSpriteRendererProperties, &NiagaraMeshRendererProperties]
  renderers [
   {rendererClass=&NiagaraSpriteRendererProperties}
   {rendererClass=&NiagaraMeshRendererProperties, rendererIndex=1}
   {rendererClass=&NiagaraSpriteRendererProperties, rendererIndex=2}
  ]
 }
 Metaball1 {
  emitterSpawnScript [/SetVariables(Emitter.ParticleAttribute=DataInterface() {propertyValues {EmitterBinding {bindingMode=Other, emitterName=Metaball1}}}) {isSetParameters=true}]
  emitterUpdateScript [
   &EmitterState()
   &SpawnRate(SpawnRate=&Multiply_Float(A=0.8, B=$System.SpawnRateScale), Use Spawn Probability=false)
   &SpawnRate(SpawnGroup=1, SpawnRate=&Multiply_Float(A=0.8, B=$System.SpawnRateScale_D), Use Spawn Probability=false) {module=SpawnRate001}
   /SetVariables(Emitter.Offset:Vector3f=<5 -20 0>) {isSetParameters=true}
   /SetVariables(Emitter.Offset_2:Vector3f=<5 -20 0>) {isSetParameters=true}
   /SetVariables(Emitter.Bound:Vector3f=&Divide_Vector(A:Vector3f=<140 140 300>, B:Vector3f=<100 100 100>)) {isSetParameters=true}
   /SetVariables(
    Emitter.Color:LinearColor=&MultiplyLinearColor(
     Linear Color=&FN_Select_Color_from_Time_Of_Day(Morning=<rgba 0.8698 0.7989 0.7835 1>, Day=<rgba 1 1 1 1>, Evening=<rgba 0.9062 0.8324 0.8163 1>, Night=<rgba 0.6146 0.6146 0.6146 1>)
     Scale Factor=&MakeVector4(X=1, Y=1, Z=1, W=&Power(a=$System.TotalScale, b=0.5))
    )
    Emitter.Position:NiagaraPosition=&AddVectorToPosition(Position=$Engine.Emitter.SimulationPosition, Vector=$Emitter.Offset)
    Emitter.Scale:Vector3f=<1 1 1>
   ) {isSetParameters=true}
   /SetVariables(Emitter.Position_2:NiagaraPosition=&AddVectorToPosition(Position=$Engine.Emitter.SimulationPosition, Vector=$Emitter.Offset_2)) {isSetParameters=true}
   &self:ReadMySelf(Particle Attribute Reader:NiagaraDataInterfaceParticleRead=$Emitter.ParticleAttribute) {module=ReadMySelf}
   &self:ReadMySelf(Particle Attribute Reader:NiagaraDataInterfaceParticleRead=$Emitter.ParticleAttribute, Particle Index=2) {module=ReadMySelf001}
   &self:ReadMySelf(Particle Attribute Reader:NiagaraDataInterfaceParticleRead=$Emitter.ParticleAttribute, Particle Index=4) {module=ReadMySelf002}
   &self:ReadMySelf(Particle Attribute Reader:NiagaraDataInterfaceParticleRead=$Emitter.ParticleAttribute, Particle Index=6) {module=ReadMySelf003}
   &self:ReadMySelf(Particle Attribute Reader:NiagaraDataInterfaceParticleRead=$Emitter.ParticleAttribute, Particle Index=8) {module=ReadMySelf004}
   &self:ReadMySelf(Particle Attribute Reader:NiagaraDataInterfaceParticleRead=$Emitter.ParticleAttribute, Particle Index=1) {module=ReadMySelf005}
   &self:ReadMySelf(Particle Attribute Reader:NiagaraDataInterfaceParticleRead=$Emitter.ParticleAttribute, Particle Index=3) {module=ReadMySelf006}
   &self:ReadMySelf(Particle Attribute Reader:NiagaraDataInterfaceParticleRead=$Emitter.ParticleAttribute, Particle Index=5) {module=ReadMySelf007}
   &self:ReadMySelf(Particle Attribute Reader:NiagaraDataInterfaceParticleRead=$Emitter.ParticleAttribute, Particle Index=7) {module=ReadMySelf008}
   &self:ReadMySelf(Particle Attribute Reader:NiagaraDataInterfaceParticleRead=$Emitter.ParticleAttribute, Particle Index=9) {module=ReadMySelf009}
   &self:IsCameraInsideBox(
    Bound:Vector3f=&Multiply_VectorByFloat(Vector=$Emitter.Bound, Float=100)
    Box Origin:Vector3f=$Emitter.Offset
    Camera Query=DataInterface() {propertyValues {PlayerControllerIndex=0, bRequireCurrentFrameData=false}}
   ) {module=IsCameraInsideBox}
   &self:IsCameraInsideBox(
    Bound:Vector3f=&Multiply_VectorByFloat(Vector=$Emitter.Bound, Float=100)
    Box Origin:Vector3f=$Emitter.Offset_2
    Camera Query=DataInterface() {propertyValues {PlayerControllerIndex=0, bRequireCurrentFrameData=false}}
   ) {module=IsCameraInsideBox001}
  ]
  particleSpawnScript [
   /SetVariables(Particles.Partition:NiagaraInt32=&ModuloInt(A=&self:GetIndexFromID(), B=2)) {enabled=false, isSetParameters=true}
   /SetVariables(Particles.RDM:NiagaraFloat=&RandomRangeFloat(Minimum=0, Maximum=1)) {isSetParameters=true}
   &InitializeParticle(
    Lifetime=&RandomRangeFloat(Minimum=4.3, Maximum=4.7)
    Sprite Size Mode=::NewEnumerator3
    Uniform Sprite Size=&Lerp_Float(A=3.5, B=7, Alpha=$Particles.RDM)
    UsePositionOffset=false
    Write Material Random=true
   )
   &ShapeLocation(
    Handle Radius=0
    Large Radius=45
    MaskSpawnGroup=false
    Offset=&MaskVectorBySpawnGroup(Masked Vector=$Emitter.Offset, Passthrough Vector=$Emitter.Offset_2)
    Offset Mode=::NewEnumerator0
    Rotation Angle:NiagaraFloat=&MakeCustomFloatFromBool(
     Bool=&SetBoolByIntComparison(A=&SpawnGroup(), B=0, Comparison Type=::NewEnumerator4)
     True Float=0
     False Float=0.5
    )
    Rotation Axis=<0 0 1>
    Rotation Mode=::NewEnumerator1
    Shape Primitive=::NewEnumerator3
    Torus Distribution Mode:ENiagaraTorusDistributionMode=::NewEnumerator1
    U Position=&Divide_Float(
     A=&Floor_Float(Float:NiagaraFloat=&Multiply_Float(A=&RandomRangeFloat(Minimum=0.3, Maximum=0.6), B=10))
     B=10
    )
    UseOwnerScale=true
   )
   &AddVelocity(Velocity=&RandomRangeVector(Minimum=<0 0 30>, Maximum=<2 2 50>))
   &self:TraceStartPos(DownOffset=128) {module=TraceStartPos}
  ]
  particleUpdateScript [
   &ParticleState()
   &Timeline(
    Initial Start Time:NiagaraFloat=&RandomRangeFloat(Minimum=10, Maximum=100)
    Play Rate=&FloatFromCurve(FloatCurve=DataInterface() {propertyValues {Curve {keys [<at 0.3968 0.4721>,<at 0.4686 1>]}}})
    Timeline End=999999
   )
   &ScaleSpriteSize(
    Uniform Curve Scale=$System.TotalScale
    Uniform Curve Sprite Scale=DataInterface() {propertyValues {Curve {
       defaultValue=1
       keys [
        {arriveTangent=9.0725, interpMode=RCIM_Cubic, leaveTangent=9.0725, tangentMode=RCTM_User, time=0, value=0.25}
        {arriveTangent=0.5491, interpMode=RCIM_Cubic, leaveTangent=0.5491, tangentMode=RCTM_User, time=0.13, value=0.9847}
        {arriveTangent=0.0679, interpMode=RCIM_Cubic, leaveTangent=0.0679, tangentMode=RCTM_User, time=0.7729, value=1}
        {arriveTangent=-3.2218, interpMode=RCIM_Cubic, leaveTangent=-3.2218, tangentMode=RCTM_User, time=1, value=-1}
       ]
      }}}
   )
   &Drag(Drag=&RandomRangeFloat(Minimum=2, Maximum=2.5), Use Linear Drag=true, Use Rotational Drag=true)
   &CurlNoiseForce(Bias Noise=false, Noise Frequency=20, Pan Noise=true, Pan Noise Field=<0 0 0.6>)
   &AccelerationForce(Acceleration=<0 0 40>)
   &SolveForcesAndVelocity(Clamp Velocity=true, Limit Acceleration=false, Speed Limit=98.2765)
   /SetVariables(
    Particles.CapHeight:NiagaraFloat=&FloatFromCurve(
     FloatCurve=DataInterface() {propertyValues {Curve {keys [
         {interpMode=RCIM_Cubic, tangentMode=RCTM_Break, time=0, value=1}
         {arriveTangent=0.2617, interpMode=RCIM_Cubic, leaveTangent=0.2617, tangentMode=RCTM_User, time=0.3, value=1}
         {arriveTangent=-1.5011, interpMode=RCIM_Cubic, tangentMode=RCTM_Break, time=0.55, value=0.01}
        ]}}}
     CurveIndex=&Multiply_Float(A=$Particles.NormalizedAge, B=1)
     Scale Curve=&Multiply_Float(A=&Lerp_Float(A=30, B=50, Alpha=$Particles.RDM), B=1)
    )
    Particles.CapRadius:NiagaraFloat=&FloatFromCurve(
     FloatCurve=DataInterface() {propertyValues {Curve {keys [
         {interpMode=RCIM_Cubic, leaveTangent=-0.2784, tangentMode=RCTM_Break, time=0.3288, value=1}
         {arriveTangent=-2.0114, tangentMode=RCTM_User, time=0.55, value=0}
        ]}}}
     CurveIndex=&Multiply_Float(A=$Particles.NormalizedAge, B=1)
     Scale Curve=&Multiply_Float(A=&MakeFloatFromVector2d(Vector 2D=$Particles.SpriteSize), B=1)
    )
    Particles.CapScaleTip:NiagaraFloat=&FloatFromCurve(
     FloatCurve=DataInterface() {propertyValues {Curve {keys [
         {interpMode=RCIM_Cubic, tangentMode=RCTM_Break, time=0.1, value=1}
         {arriveTangent=-6.6294, interpMode=RCIM_Cubic, leaveTangent=-6.6294, tangentMode=RCTM_User, time=0.2828, value=0.5289}
         <at 0.5 0>
        ]}}}
     CurveIndex=&Multiply_Float(A=$Particles.NormalizedAge, B=1)
    )
   ) {isSetParameters=true}
   /SetVariables(Particles.PositionAndSizeInfo:Vector4f=&MakeVector4FromVectorAndFloat(
     "Vector (RGB)"=&TransformPosition(Position=$Particles.Position, Source Space=::Simulation, Destination Space=::World, Apply Scale=true)
     "Float (Alpha)"=&MakeFloatFromVector2d(Vector 2D=$Particles.SpriteSize)
    )) {isSetParameters=true}
   /SetVariables(Particles.CapsuleInfo:Vector4f=&MakeVector4(X=$Particles.CapHeight, Y=$Particles.CapRadius, Z=$Particles.CapScaleTip, W=$Particles.CapsuleOffset)) {isSetParameters=true}
   &self:SecondParticleMotion(
    Age:NiagaraFloat=$Particles.Timeline.Time
    Offset:NiagaraFloat=&Multiply_Float(
     A=&RandomRangeFloat(Minimum=0.6, Maximum=0.9)
     B=&MakeFloatFromVector2d(Vector 2D=$Particles.SpriteSize)
    )
    RandomZ:Vector3f=&NormalizeVector(
     Vector To Normalize=&RandomRangeVector(Minimum=<-1 -1 -1>, Maximum=<1 1 1>)
     Safe Normalize=true
     Safe Fallback Threshold=0
     Safe Fallback Vector=<0 0 0.001>
    )
    Speed:NiagaraFloat=&RandomRangeFloat(Minimum=100, Maximum=150)
   ) {module=SecondParticleMotion}
   /SetVariables(Particles.SecondBlobInfo:Vector4f=&MakeVector4(
     X=&MakeFloatFromVector(
      Vector=&TransformPosition(Position=$Particles.SecondParticleMotion.SecBPosition, Source Space=::Simulation, Destination Space=::World)
      Channel=::NewEnumerator0
     )
     Y=&MakeFloatFromVector(
      Vector=&TransformPosition(Position=$Particles.SecondParticleMotion.SecBPosition, Source Space=::Simulation, Destination Space=::World)
      Channel=::NewEnumerator1
     )
     Z=&MakeFloatFromVector(
      Vector=&TransformPosition(Position=$Particles.SecondParticleMotion.SecBPosition, Source Space=::Simulation, Destination Space=::World)
      Channel=::NewEnumerator2
     )
     W=&Multiply_Float(
      A=&MakeFloatFromVector2d(Vector 2D=$Particles.SpriteSize)
      B=&RandomRangeFloat(Minimum=0.9, Maximum=1.1)
     )
    )) {isSetParameters=true}
   /SetVariables(Particles.GroundInfo:Vector4f=&MakeVector4(
     X=0
     Y=0
     Z=1
     W=&MakeFloatFromVector(
      Vector=&AddVectorToPosition(
       Position=$Particles.StartPos
       Vector=&MakeVector(
        X=0
        Y=0
        Z=&FloatFromCurve(FloatCurve=DataInterface() {propertyValues {Curve {keys [<at 0.2315 0>,<at 0.55 -1>]}}}, Scale Curve=50)
       )
      )
      Channel=::NewEnumerator2
     )
    )) {isSetParameters=true}
   /SetVariables(Particles.PriteSizeForSoftGlow:Vector2f=&Multiply_Vector2D(A=$Particles.SpriteSize, B=<7 7>)) {isSetParameters=true}
   &ScaleColor(
    Linear Color Curve=DataInterface() {propertyValues {
      AlphaCurve {keys [<at 0 0>,<at 0.1244 0.4088>,<at 0.4966 0.9356>,<at 0.8403 0.4726>,<at 1 0>]}
      BlueCurve {keys [<at 0 1>,<at 1 1>]}
      GreenCurve {keys [<at 0 1>,<at 1 1>]}
      RedCurve {keys [<at 0 1>,<at 1 1>]}
     }}
    Scale Mode=2
   )
  ]
  rendererClasses [&NiagaraMeshRendererProperties, &NiagaraSpriteRendererProperties]
  renderers [
   {rendererClass=&NiagaraMeshRendererProperties, rendererIndex=0}
   {rendererClass=&NiagaraMeshRendererProperties, rendererIndex=1}
   {rendererClass=&NiagaraSpriteRendererProperties, rendererIndex=2}
   {rendererClass=&NiagaraMeshRendererProperties, rendererIndex=3}
   {rendererClass=&NiagaraMeshRendererProperties, rendererIndex=4}
  ]
 }
 SmallBubble {
  emitterSpawnScript [
   /SetVariables() {isSetParameters=true}
   /SetVariables(Emitter.Particle Attribute Reader=DataInterface() {propertyValues {EmitterBinding {bindingMode=Other, emitterName=Metaball1}}}) {isSetParameters=true}
  ]
  emitterUpdateScript [
   &EmitterState()
   &self:SpawnFromMather(Particle Attribute Reader:NiagaraDataInterfaceParticleRead=$Emitter.Particle Attribute Reader) {module=SpawnFromMather}
   &SpawnPerFrame(
    Spawn Count:NiagaraInt32=&Multiply_Int(
     A=$Emitter.ThisFrameSpawned
     B=&MakeCustomIntFromBool(
      Bool=&IsPlatformSetActive(New Platform Set=DataInterface() {propertyValues {Platforms {
          qualityLevelMask=-1
          deviceProfileStates [
           {profileName=Mobile, qualityLevelMask=0, setQualityLevelMask=4294967295}
           {profileName=Switch, qualityLevelMask=0, setQualityLevelMask=4294967295}
           {profileName=Switch2, qualityLevelMask=0, setQualityLevelMask=4294967295}
           {profileName=XB1, qualityLevelMask=0, setQualityLevelMask=4294967295}
           {profileName=PS4, qualityLevelMask=0, setQualityLevelMask=4294967295}
          ]
         }}})
      True INT=4
      False INT=3
     )
    )
    Use Spawn Probability=false
   )
  ]
  particleSpawnScript [
   &PartitionParticles(Number Of Partitions=$Emitter.ThisFrameSpawned)
   &InitializeParticle(
    Color=&MergeHelper_ColorSwitch(
     0=&RandomRangeLinearColor(Minimum=<rgba 0 0.8333 1 1>, Maximum=<rgba 0 0.3333 1 1>)
     1=&RandomRangeLinearColor(Minimum=<rgba 0 1 0.3333 1>, Maximum=<rgba 0 1 0.5 1>)
     2=&RandomRangeLinearColor(Minimum=<rgba 0.3333 0 1 1>, Maximum=<rgba 0.5 0 1 1>)
     3=&RandomRangeLinearColor(Minimum=<rgba 0.2 0 0 1>, Maximum=<rgba 0.2 0 0 1>)
     4=<rgba 0 0 0 0>
     Index=$User.ColorIndex
     Num=::NewEnumerator3
    )
    Lifetime=&RandomRangeFloat(Minimum=2, Maximum=4.5)
    Sprite Size=&Multiply_Vector2D_ByFloat(
     Float=&RandomRangeFloat(Minimum=1, Maximum=5)
     Vector2D=&RandomRangeVector2D(Minimum=<1 1>, Maximum=<1 1.3>)
    )
    Sprite Size Mode=::NewEnumerator4
    UsePositionOffset=false
    Write Material Random=true
   )
   &self:ScratchModule_01(
    PartID:NiagaraInt32=$Particles.PartitionParticles.Partition
    Particle Attribute Reader:NiagaraDataInterfaceParticleRead=$Emitter.Particle Attribute Reader
   ) {module=ScratchModule_01}
   &ShapeLocation(
    MaskSpawnGroup=false
    Non Uniform Scale=<0.5 0.5 1>
    Offset=<0 0 -20>
    Offset Mode=::NewEnumerator0
    Override Seed=false
    Sphere Radius=4
    UseOwnerScale=true
   )
  ]
  particleUpdateScript [
   &ParticleState()
   &SpriteRotationRate(Rotation Rate=&RandomRangeFloat(Minimum=-300, Maximum=300))
   &ScaleSpriteSize(
    Uniform Curve Scale=$System.TotalScale
    Uniform Curve Sprite Scale=DataInterface() {propertyValues {Curve {
       defaultValue=0
       keys [
        {arriveTangent=9.0725, interpMode=RCIM_Cubic, leaveTangent=9.0725, tangentMode=RCTM_User, time=0, value=0}
        {arriveTangent=-0.06, interpMode=RCIM_Cubic, leaveTangent=-0.06, tangentMode=RCTM_User, time=0.1408, value=0.9847}
        {arriveTangent=-1.146, interpMode=RCIM_Cubic, leaveTangent=-1.146, time=0.7505, value=0.8179}
        {arriveTangent=-2.6043, interpMode=RCIM_Cubic, leaveTangent=-2.6043, tangentMode=RCTM_User, time=1, value=0}
       ]
      }}}
   )
   &self:ScratchModule_02(
    Niagara ID:NiagaraID=$Particles.SpawnID
    Offet:Vector3f=&AddVectorToPosition(
     Position=$Particles.ShapeLocation.ShapePosition
     Vector=&MakeVector(
      X=0
      Y=0
      Z=&FloatFromCurve(
       FloatCurve=DataInterface() {propertyValues {Curve {keys [{interpMode=RCIM_Cubic, tangentMode=RCTM_Break, time=0, value=0.1}, <at 1 0.9>]}}}
       Scale Curve=&RandomRangeFloat(Minimum=20, Maximum=25)
      )
     )
    )
    Particle Attribute Reader:NiagaraDataInterfaceParticleRead=$Emitter.Particle Attribute Reader
    SampleID:NiagaraInt32=$Particles.SampleID
   ) {module=ScratchModule_02}
   &ScaleColor(
    Scale Alpha=&FloatFromCurve(
     FloatCurve=DataInterface() {propertyValues {Curve {keys [
         {interpMode=RCIM_Cubic, tangentMode=RCTM_Break, time=0, value=0}
         {arriveTangent=8.998, interpMode=RCIM_Cubic, tangentMode=RCTM_Break, time=0.3, value=1}
        ]}}}
     Scale Curve=$System.TotalScale
    )
    Scale RGB=&VectorFromFloat(Value=&FloatFromCurve(
      FloatCurve=DataInterface() {propertyValues {Curve {keys [
          {interpMode=RCIM_Cubic, tangentMode=RCTM_Break, time=0, value=0}
          {arriveTangent=8.998, interpMode=RCIM_Cubic, tangentMode=RCTM_Break, time=0.3, value=1}
         ]}}}
      Scale Curve=&FloatFromCurve(
       FloatCurve=DataInterface() {propertyValues {Curve {keys [<at 1 50>,<at 2 1.5>,<at 4 1>]}}}
       CurveIndex=&MakeFloatFromVector2d(Vector 2D=$Particles.SpriteSize)
       Scale Curve=$System.TotalScale
      )
     ))
    ScaleA=true
    ScaleRGB=true
   )
  ]
  rendererClasses [&NiagaraSpriteRendererProperties]
  renderers [{rendererClass=&NiagaraSpriteRendererProperties}]
 }
}
summary {
 systemName=NS_Emote_BandSoda_Metaballs
 emitters [
  {emitterName=SmallBubble, rendererClasses [&NiagaraSpriteRendererProperties]}
  {emitterName=Metaball1, rendererClasses [&NiagaraMeshRendererProperties, &NiagaraSpriteRendererProperties]}
  {emitterName=FallBack, rendererClasses [&NiagaraSpriteRendererProperties, &NiagaraMeshRendererProperties]}
 ]
 userVariables [User.ColorIndex {type=NiagaraInt32}]
}
