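# Niagara system for the spawn splash effect of the BandSoda emote in CosmeticCompanions, built from Splashes and Droplets emitters.
# It exposes an integer User.ColorIndex, alongside dynamic inputs that select a color from an array, multiply it and vary it by time of day.
# A Set Variables assignment in the system spawn script and SystemState are the listed modules.
# Neither emitter lists any renderers.
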NiagaraSystem /CosmeticCompanions/FX/Builtins/BandSoda/Emote/NS_Emote_BandSoda_SpawnSplash
dependencies {
 usedDataInterfaces [&NiagaraDataInterfaceArrayColor]
 usedDynamicInputs [
  &MultiplyLinearColor
  &SelectLinearColorFromArray
  &MakeVector4_FromLinearColor
  &FN_Select_Color_from_Time_Of_Day
 ]
 usedModules [
  &self:SystemSpawnScript.SystemScriptSource.SystemScriptGraph.NiagaraNodeAssignment_0.SetVariables_#1
  &SystemState
 ]
}
emitters {Droplets {rendererClasses [], renderers []}, Splashes {rendererClasses [], renderers []}}
summary {
 systemName=NS_Emote_BandSoda_SpawnSplash
 emitters [{emitterName=Splashes, rendererClasses []}, {emitterName=Droplets, rendererClasses []}]
 userVariables [User.ColorIndex {type=NiagaraInt32}]
}
